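# Translucent UI material for the Captain America (Ashton) shield reticle, turning red when an enemy is in range.
# It builds SDF circle outlines, two radial segments spaced by ActivateShield, an angle-gradient progress ring, and a star that fills or morphs into a line.
# A time-seeded random grid supplies a pixelated effect.
# Named reroutes feed these masks and their shadows into composites weighted by opacity parameters, including shield break.
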Material /AshtonGameplay/Assets/CaptainAmerica/Shield/UI/M_UI_Ashton_Reticle
graph {
 properties {bAutomaticallySetUsageInEditor=false, blendMode=BLEND_Translucent, materialDomain=MD_UI, usedShadingModels=Default Lit}
 nodes [
  Constant() {r=0.5}
  Constant() {r=0.27}
  Constant() {r=0.17}
  Constant() {r=0.95}
  n1 = Multiply(A=NamedRerouteUsage(), B=NamedRerouteUsage())
  n2 = NamedRerouteUsage()
  n3 = Multiply(A=NamedRerouteUsage(), B=ScalarParameter() {parameterName=ProgressBarOpacity})
  n4 = Saturate(None=Subtract(A=NamedRerouteUsage(), B=NamedRerouteUsage()))
  LinearInterpolate(
   A=Constant4Vector().RGBA
   B=LinearInterpolate(
    A=Constant4Vector() {constant=<rgba 1 1 1 1>}.RGBA
    B=Constant4Vector() {constant=<rgba 1 0 0 1>}.RGBA
    Alpha=ScalarParameter() {parameterName=EnemyInRange}
   )
   Alpha=Add(A=Add(A=@n1, B=@n2), B=Multiply(A=@n3, B=@n4))
  )
  NamedRerouteDeclaration(None=MaterialFunctionCall(UVs=TextureCoordinate(), Texture Scale=ScalarParameter() {defaultValue=1, parameterName=CircleSize}) {materialFunction=&ScaleUVsByCenter}.UVs) {name=UV, nodeColor=<rgba 0 0 0 1>}
  n5 = ScalarParameter() {parameterName=ActivateShield}
  n6 = MaterialFunctionCall(UVs=NamedRerouteUsage(), "Rotation Angle (0-1)"=Constant() {r=0.25}) {materialFunction=&CustomRotator}
  n7 = Constant() {r=2}
  NamedRerouteDeclaration(None=SmoothStep(
    Max=Multiply(B=@n5) {constA=0.2}
    Value=MaterialFunctionCall!(UV=@n6.Rotated Values, Segment Count=@n7, Segment Spacing=@n5, Fixed Spacing) {materialFunction=/FortUILibrary/MaterialFunctions/Basics/MF_UI_RadialSegments}.Segmented Triangles Gradient
   ) {constMax=0.1}) {name=Segments, nodeColor=<rgba 0 0 0 1>}
  n8 = MaterialFunctionCall() {materialFunction=&VectorToRadialValue}
  n9 = Constant() {r=16}
  n10 = Constant()
  n11 = Time()
  n12 = Constant() {r=0.8}
  n13 = Saturate(None=MaterialFunctionCall!(
    UV=MaterialFunctionCall!(
     UV=@n8.Vector Converted to Angle
     Global Offset
     Global Rotation
     Tiling=@n9
     Cell Offset=@n10
     Cell Scale=@n10
     Cell Rotation=@n10
     Random Seed
    ) {materialFunction=/FortUILibrary/MaterialFunctions/Basics/MF_UI_ParticleGrid}.Random Cell Values
    Seed=Multiply(A=@n11, B=@n12)
   ) {materialFunction=&MF_UI_Random}.Result)
  NamedRerouteDeclaration(None=Clamp(None=@n13, Min=ScalarParameter() {defaultValue=1, parameterName=PixelFxMin}) {minDefault=1}) {name=PixelatedFx, nodeColor=<rgba 0 0 0 1>}
  n14 = MaterialFunctionCall(UVs=NamedRerouteUsage(), Size=Constant() {r=0.92}, OutlineThickness=ScalarParameter() {defaultValue=0.01, parameterName=OutlineThickness}) {materialFunction=&MF_UI_Circle}
  n15 = SmoothStep(Value=@n14.SDFOutline) {constMax=0.04}
  NamedRerouteDeclaration(None=Multiply(A=OneMinus(None=@n15), B=NamedRerouteUsage())) {name=CircleOutline, nodeColor=<rgba 0 0 0 1>}
  n16 = NamedRerouteUsage()
  n17 = Constant() {r=0.84}
  n18 = Constant() {r=0.09}
  n19 = OneMinus(None=SmoothStep(Value=MaterialFunctionCall(UVs=@n16, Size=@n17, OutlineThickness=@n18) {materialFunction=&MF_UI_Circle}.SDFOutline) {constMax=0.01})
  n20 = NamedRerouteUsage()
  NamedRerouteDeclaration(None=Saturate(None=Subtract(A=@n19, B=@n20))) {name=ProgressBarBase, nodeColor=<rgba 0 0 0 1>}
  n21 = ScalarParameter() {parameterName=ActivateShield}
  n22 = NamedRerouteUsage()
  n23 = Constant() {r=0.25}
  n24 = SmoothStep(
   Max=Multiply(B=@n21) {constA=0.2}
   Value=MaterialFunctionCall(
    UV=MaterialFunctionCall(UVs=@n22, "Rotation Angle (0-1)"=@n23) {materialFunction=&CustomRotator}.Rotated Values
    Segment Count=Constant() {r=2}
    Segment Spacing=@n21
    Fixed Spacing=Constant() {r=1}
   ) {materialFunction=&MF_UI_RadialSegments}.Segmented Triangles Gradient
  )
  n25 = NamedRerouteUsage()
  NamedRerouteDeclaration(None=Saturate(None=Add(A=@n24, B=@n25))) {name=SegmentsShadow, nodeColor=<rgba 0 0 0 1>}
  n26 = NamedRerouteUsage()
  n27 = Subtract(
   A=OneMinus(None=MaterialFunctionCall!(UVs=@n26, Rotation) {materialFunction=/FortUILibrary/MaterialFunctions/Gradients/MF_UI_AngleGradient}.Result)
   B=ScalarParameter() {defaultValue=0.5, parameterName=ProgressBar}
  )
  NamedRerouteDeclaration(None=OneMinus(None=SmoothStep(Value=@n27) {constMax=0.002})) {name=Progress, nodeColor=<rgba 0 0 0 1>}
  n28 = SmoothStep(Value=@n14.SDFOutline) {constMax=0.05}
  NamedRerouteDeclaration(None=Multiply(A=OneMinus(None=@n28), B=NamedRerouteUsage())) {name=CircleOutlineShadow, nodeColor=<rgba 0 0 0 1>}
  n29 = TextureCoordinate()
  n30 = Constant2Vector() {g=0.2, r=0.2}
  n31 = ScalarParameter() {defaultValue=1, parameterName=StarHeight}
  n32 = Multiply(A=@n30.G, B=@n31)
  n33 = TextureSample(UVs=MaterialFunctionCall(UVs=@n29, Texture Scale=AppendVector(A=@n30.R, B=@n32)) {materialFunction=&ScaleUVsByCenter}.UVs) {
   samplerSource=SSM_Clamp_WorldGroupSettings
   samplerType=SAMPLERTYPE_LinearGrayscale
   texture=&/T_UI_AshtonShield_Star_Outline
  }
  n34 = MaterialFunctionCall!(RGBA Channels=@n33.RGBA, Glow Min=Constant() {r=0.35}, Glow Max=Constant() {r=0.9}, Outline Thickness) {materialFunction=/FortUILibrary/MaterialFunctions/Basics/MF_UI_ChannelPackedDistanceField}
  n35 = Constant2Vector() {g=0.12, r=0.12}
  n36 = Multiply(A=@n31, B=@n35.G)
  n37 = TextureSample(UVs=MaterialFunctionCall(UVs=@n29, Texture Scale=AppendVector(A=@n35.R, B=@n36)) {materialFunction=&ScaleUVsByCenter}.UVs) {
   samplerSource=SSM_Clamp_WorldGroupSettings
   samplerType=SAMPLERTYPE_LinearGrayscale
   texture=&/T_UI_AshtonShield_Star
  }
  n38 = Multiply(
   A=MaterialFunctionCall(RGBA Channels=@n37.RGBA, Glow Min=Constant(), Glow Max=Constant() {r=0.1}) {materialFunction=&MF_UI_ChannelPackedDistanceField}.Fill
   B=ScalarParameter() {parameterName=FillStar}
  )
  n39 = MaterialFunctionCall(UVs=NamedRerouteUsage(), Size=Constant2Vector() {g=0.03, r=0.23}, OutlineThickness=Constant() {r=0.01}) {materialFunction=&MF_UI_Box}
  n40 = ScalarParameter() {parameterName=ChangeStarToLine}
  NamedRerouteDeclaration(None=LinearInterpolate(A=Add(A=@n34.Fill, B=@n38), B=@n39.Fill, Alpha=@n40) {constAlpha=0}) {name=Star, nodeColor=<rgba 0 0 0 1>}
  n41 = Constant()
  n42 = Constant() {r=1}
  n43 = Subtract(A=@n39.Outline, B=@n39.Outline)
  NamedRerouteDeclaration(None=LinearInterpolate(
    A=MaterialFunctionCall(RGBA Channels=@n33.RGBA, Glow Min=@n41, Glow Max=@n42) {materialFunction=&MF_UI_ChannelPackedDistanceField}.Fill
    B=Saturate(None=@n43)
    Alpha=@n40
   )) {name=StarShadow, nodeColor=<rgba 0 0 0 1>}
  Power(Base=@n13) {constExponent=3}
  n44 = NamedRerouteUsage()
  n45 = NamedRerouteUsage()
  n46 = LinearInterpolate(Alpha=ScalarParameter() {parameterName=Outer Circle Opacity})
  n47 = ScalarParameter() {parameterName=ShieldBreakOpacity}
  n48 = Multiply(A=NamedRerouteUsage(), B=NamedRerouteUsage())
  n49 = LinearInterpolate(
   A=LinearInterpolate(Alpha=ScalarParameter() {parameterName=Outer Circle Opacity, sliderMax=1})
   Alpha=ScalarParameter() {parameterName=ShieldBreakOpacity}
  ) {constB=0}
  n50 = Add(
   A=Multiply(A=Multiply(A=@n44, B=@n45), B=LinearInterpolate(A=@n46, Alpha=@n47) {constB=0})
   B=Multiply(A=Multiply(A=@n48, B=@n49), B=ScalarParameter() {defaultValue=0.5, parameterName=Shadow Opacity})
  )
  n51 = Multiply(A=Multiply(A=NamedRerouteUsage()) {constB=0.4}, B=ScalarParameter() {parameterName=ProgressBarOpacity})
  n52 = Add(A=Multiply(A=NamedRerouteUsage(), B=ScalarParameter() {defaultValue=0.5, parameterName=Shadow Opacity}), B=NamedRerouteUsage())
  n53 = ScalarParameter() {defaultValue=1, parameterName=StarOpacity}
  Saturate(None=Add(A=Add(A=@n50, B=@n51), B=Multiply(A=@n52, B=@n53)))
 ]
}
