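# Reticle extension widget for Captain America's shield weapon, showing blocking, block health, shield break, throw and valid-target states on one material-driven image.
# It binds to the weapon, its blocking volume and target-state component, and listens for throw and catch gameplay events.
# Remaining block health drives the material's progress bar and the stagger or broken state.
# Delta gates limit widget animations, mostly material-parameter tracks, to state changes, with a timer after breaks.
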WidgetBlueprint /AshtonGameplay/Assets/CaptainAmerica/Shield/UI/WBP_Ashton_Shield_Reticle
animations [
 Anim_ShieldActivated {
  binding_count=1
  length_seconds=0.3667
  object=&self:Anim_ShieldActivated
  sequence {
   bindings [M_UI_Reticle {
     id=#1
     possessed_class=Image
     tracks [{
       type=MovieSceneWidgetMaterialTrack
       sections [{
         type=MovieSceneParameterSection
         end_frame=11
         start_frame=0
         channels [
          CircleSize {type=MovieSceneScriptingFloatChannel, const [1,2]}
          Outer Circle Opacity {type=MovieSceneScriptingFloatChannel, const [1,2]}
          ActivateShield {type=MovieSceneScriptingFloatChannel, keys [[0,1],[11,0]]}
         ]
        }]
      }]
    }]
   display_rate {numerator=30}
  }
 }
 Anim_StaggerMeter {
  binding_count=1
  length_seconds=0.25
  object=&self:Anim_StaggerMeter
  sequence {bindings [M_UI_Reticle {
     id=#2
     possessed_class=Image
     tracks [{
       type=MovieSceneWidgetMaterialTrack
       sections [{type=MovieSceneParameterSection, channels [ProgressBarOpacity {type=MovieSceneScriptingFloatChannel, keys [[0,0],[5,1]]}]}]
      }]
    }]}
 }
 Anim_TrownShield {
  binding_count=1
  length_seconds=0.244
  object=&self:Anim_TrownShield
  sequence {bindings [M_UI_Reticle {
     id=#3
     possessed_class=Image
     tracks [{
       type=MovieSceneFloatTrack
       display_name=Render Opacity
       sections [{type=MovieSceneFloatSection, end_frame=4, start_frame=0, channels [None {type=MovieSceneScriptingFloatChannel, keys [[0,1],[4.8803,0.3]]}]}]
      }]
    }]}
 }
 Anim_TrownShield_02 {
  binding_count=1
  length_seconds=0.2366
  object=&self:Anim_TrownShield_02
  sequence {bindings [M_UI_Reticle {
     id=#3
     possessed_class=Image
     tracks [{
       type=MovieSceneWidgetMaterialTrack
       sections [{
         type=MovieSceneParameterSection
         end_frame=4
         start_frame=0
         channels [
          ChangeStarToLine {type=MovieSceneScriptingFloatChannel, keys [[0,0],[2.4973,0],[4.6613,1]]}
          StarHeight {type=MovieSceneScriptingFloatChannel, keys [[0,1],[4.696,0.3]]}
         ]
        }]
      }]
    }]}
 }
 Anim_BrokenShield {
  binding_count=1
  length_seconds=0.5503
  object=&self:Anim_BrokenShield
  sequence {bindings [M_UI_Reticle {
     id=#4
     possessed_class=Image
     tracks [{
       type=MovieSceneWidgetMaterialTrack
       sections [{
         type=MovieSceneParameterSection
         end_frame=11
         start_frame=0
         channels [
          CircleSize {type=MovieSceneScriptingFloatChannel, keys [[0,1],[10,1.1],[11,1]]}
          Outer Circle Opacity {type=MovieSceneScriptingFloatChannel, const [1,1]}
          ActivateShield {type=MovieSceneScriptingFloatChannel, const [0,1]}
          ProgressBarOpacity {type=MovieSceneScriptingFloatChannel, keys [[0,1],[5,0]]}
          OutlineThickness {type=MovieSceneScriptingFloatChannel, keys [[0,0.01],[10,0.035],[11,0],[11.0063,0.01]]}
          StarOpacity {type=MovieSceneScriptingFloatChannel, keys [[0,1],[5,1],[10,0]]}
          ShieldBreakOpacity {type=MovieSceneScriptingFloatChannel, keys [[0,0],[3,0],[10,1]]}
          BreakMin {type=MovieSceneScriptingFloatChannel, keys [[0,0.9],[10,0],[11,0]]}
          BreakMax {type=MovieSceneScriptingFloatChannel, keys [[0,1],[10,0.1],[11,50]]}
          PixelFxMin {type=MovieSceneScriptingFloatChannel, keys [[0.0163,1],[0.7007,1],[5.067,0],[11.0063,1]]}
         ]
        }]
      }]
    }]}
 }
 Anim_EnemyInRange {
  binding_count=1
  length_seconds=0.1505
  object=&self:Anim_EnemyInRange
  sequence {bindings [M_UI_Reticle {
     id=#5
     possessed_class=Image
     tracks [{
       type=MovieSceneWidgetMaterialTrack
       sections [{
         type=MovieSceneParameterSection
         end_frame=3
         start_frame=0
         channels [FillStar {type=MovieSceneScriptingFloatChannel, keys [[0,0],[3.009,1]]}]
        }]
      }]
    }]}
 }
 Anim_Init {
  binding_count=1
  length_seconds=0.2
  object=&self:Anim_Init
  sequence {bindings [M_UI_Reticle {
     id=#6
     possessed_class=Image
     tracks [{
       type=MovieSceneWidgetMaterialTrack
       sections [{
         type=MovieSceneParameterSection
         end_frame=4
         start_frame=0
         channels [
          StarOpacity {type=MovieSceneScriptingFloatChannel, keys [[0,0],[4,1]]}
          ShieldBreakOpacity {type=MovieSceneScriptingFloatChannel, keys [[0,1],[4,0]]}
         ]
        }]
      }]
    }]}
 }
]
graphs [
 EventGraph {nodes [
   n1 = StaggerUpdated() {is=title K2Node_CustomEvent}
   > UpdateForStagger()
   n2 = OnBlockingStateConfirmed!() -> (OutputDelegate, then, NewActionState:EBallisticShieldPlayerActionState Enum=Idling) {is=title K2Node_CustomEvent}
   > Set IsBlockingConfirmed?(IsBlockingConfirmed?:bool=Select!(Idling:bool=false, Blocking:bool=true, Charging:bool=true, Index:EBallisticShieldPlayerActionState Enum=@n2.NewActionState) -> (ReturnValue:bool)) -> (Output_Get:bool)
   > UpdateStaggerMeterVisibility!(execute, self, IsInstantHide:bool=true) -> (then)
   n3 = OnActionStateChanged!() -> (
    OutputDelegate
    then
    PreviousActionState:EBallisticShieldPlayerActionState Enum=Idling
    NewActionState:EBallisticShieldPlayerActionState Enum=Idling
   ) {is=title K2Node_CustomEvent}
   > SetShieldBlockingState!(execute, self, IsBlocking?:bool=Select(Idling=false, Blocking=true, Charging=true, Index=@n3.NewActionState)) -> (then)
   n4 = UpdateBlockingBGA() {is=title K2Node_CustomEvent}
   > UnbindBlockingBGA()
   > n5 = Set Blocking BGA(Blocking BGA:BGA Ashton Shield Block Volume Object Reference=Get BlockingBGA!(self:B Ashton Shield Weap Object Reference=Get As B Ashton Shield Weap() -> (As B Ashton Shield Weap:B Ashton Shield Weap Object Reference)) -> (BlockingBGA:BGA Ashton Shield Block Volume Object Reference)) -> (Output_Get:BGA Ashton Shield Block Volume Object Reference)
   > Is Valid(InputObject=@n5)
   > Bind Event to On Stagger Updated(self:BGA Ashton Shield Block Volume Object Reference=@n5, Delegate=@n1) {is=title K2Node_AddDelegate}
   > UpdateForStagger()
   Event Construct()
   n6 = GetOwningPlayer()
   > Is Valid(InputObject=@n6)
   > n7 = Cast To PlayerPawn_Athena(Object=GetViewTarget!(self:Controller Object Reference=@n6) -> (ReturnValue:Actor Object Reference))
   > Set As Player Pawn Athena(As Player Pawn Athena:Player Pawn Athena Object Reference=@n7) -> (Output_Get:Player Pawn Athena Object Reference)
   > n8 = Cast To B_Ashton_Shield_Weap!(
    execute
    Object=GetEquippedWeapon!(self:Fort Weapon Reticle Extension Widget Base Object Reference) -> (ReturnValue:Fort Weapon Object Reference)
   ) -> (then, CastFailed, AsB Ashton Shield Weap:B Ashton Shield Weap Object Reference) {is=title K2Node_DynamicCast}
   > n9 = Set As B Ashton Shield Weap(As B Ashton Shield Weap:B Ashton Shield Weap Object Reference=@n8) -> (Output_Get:B Ashton Shield Weap Object Reference)
   > Is Valid(InputObject=@n9)
   > Bind Event to Blocking BGA Is Set(self:B Ashton Shield Weap Object Reference=@n9, Delegate=@n4) {is=title K2Node_AddDelegate}
   > Bind Event to On Action State Changed(self:B Ashton Shield Weap Object Reference=Get As B Ashton Shield Weap(), Delegate=@n3) {is=title K2Node_AddDelegate}
   > Bind Event to On Blocking State Confirmed(self:B Ashton Shield Weap Object Reference=Get As B Ashton Shield Weap(), Delegate=@n2) {is=title K2Node_AddDelegate}
   > n10 = Set TargetStateComponent(TargetStateComponent:Fort Weapon Component Target State Object Reference=GetComponentByClass!(
     self:Actor Object Reference=Get As B Ashton Shield Weap()
     ComponentClass:Actor Component Class Reference=/Script/FortniteGame.FortWeaponComponent_TargetState
    ) -> (ReturnValue:Fort Weapon Component Target State Object Reference)) -> (Output_Get:Fort Weapon Component Target State Object Reference)
   > Bind Event to On Is Valid Target Changed(self:Fort Weapon Component Target State Object Reference=@n10, Delegate=Create Event()) {is=title K2Node_AddDelegate}
   > n11 = WaitGameplayEventToActor!(
    execute:Exec
    TargetActor:Actor Object Reference=Get As Player Pawn Athena() -> (As Player Pawn Athena:Player Pawn Athena Object Reference)
    EventTag:Gameplay Tag Structure=(TagName="Event.Athena.Ashton.Shield.Throw")
    OnlyTriggerOnce:bool=false
    OnlyMatchExact:bool=true
   ) -> (then:Exec, AsyncTaskProxy:Ability Async Wait Gameplay Event Object Reference, EventReceived:Exec, Payload:Gameplay Event Data Structure) {is=title K2Node_AsyncAction}
   > Add(
    TargetArray:Array of Ability Async Object References=Get AsyncActions() -> (AsyncActions:Array of Ability Async Object References)
    NewItem:"Ability Async Object Reference (by ref)"=@n11
   ) {is=title K2Node_CallArrayFunction}
   > n12 = WaitGameplayEventToActor(TargetActor=Get As Player Pawn Athena(), EventTag=(TagName="Event.Ability.Ashton.Shield.Catch"), OnlyTriggerOnce=false, OnlyMatchExact=true)
   > Add(TargetArray=Get AsyncActions(), NewItem=@n12)
   PlayAnimationReverse(execute=@n12.EventReceived, InAnimation=Get Anim_TrownShield() -> (Anim_TrownShield:Widget Animation Object Reference))
   PlayAnimationForward(execute=@n11.EventReceived, InAnimation=Get Anim_TrownShield())
   Event Destruct()
   > Parent: Destruct()
   > UnbindBlockingBGA()
   > EndAction(self:Ability Async Object Reference=Get AsyncActions())
   > n13 = Sequence()
   > n14 = Get!(execute:Exec) -> (then:Exec, else:Exec, As B Ashton Shield Weap:B Ashton Shield Weap Object Reference)
   > Unbind Event from Blocking BGA Is Set(self:B Ashton Shield Weap Object Reference=@n14, Delegate=Create Event()) {is=title K2Node_RemoveDelegate}
   > Unbind Event from On Action State Changed(self:B Ashton Shield Weap Object Reference=@n14, Delegate=Create Event()) {is=title K2Node_RemoveDelegate}
   n15 = Get!(execute=@n13.then_1) -> (then, else, TargetStateComponent:Fort Weapon Component Target State Object Reference)
   > Unbind Event from On Is Valid Target Changed(self:Fort Weapon Component Target State Object Reference=@n15, Delegate=Create Event()) {is=title K2Node_RemoveDelegate}
   Unbind Event from On Blocking State Confirmed(self:B Ashton Shield Weap Object Reference=@n14, Delegate=Create Event()) {is=title K2Node_RemoveDelegate}
  ]}
 GetProgress {nodes [
   GetProgress() {is=title K2Node_FunctionEntry}
   n1 = Get Blocking BGA() -> (Blocking BGA:BGA Ashton Shield Block Volume Object Reference)
   n2 = SafeDivide(
    A=Get DamageCount!(self:BGA Ashton Shield Block Volume Object Reference=Get Blocking BGA()) -> (DamageCount:double)
    B=GetValueAtLevel!(
     Input:"Scalable Float Structure (by ref)"=Get HF_BlockHealth!(self:BGA Ashton Shield Block Volume Object Reference=@n1) -> (HF_BlockHealth:Scalable Float Structure)
     Level:float=0
    ) -> (ReturnValue:float)
   )
   n3 = Make Literal Float(Value=1)
   > Return Node!(execute, Progress:double="Clamp (Float)"(Value=float - float(A=@n3, B=@n2), Min=0, Max=1))
  ]}
 HideStaggerUI {nodes [HideStaggerUI() {is=title K2Node_FunctionEntry}]}
 OnBreakOver {nodes [
   OnBreakOver() {is=title K2Node_FunctionEntry}
   > PlayAnimation(InAnimation=Get Anim_Init() -> (Anim_Init:Widget Animation Object Reference))
  ]}
 OnIsValidTargetChanged {nodes [
   n1 = OnIsValidTargetChanged!() -> (then, bIsValidTarget:bool=~) {is=title K2Node_FunctionEntry}
   > n2 = Branch(Condition=@n1)
   n3 = Get Anim_EnemyInRange() -> (Anim_EnemyInRange:Widget Animation Object Reference)
   > PlayAnimationForward(InAnimation=@n3)
   PlayAnimationReverse(execute=@n2.else, InAnimation=@n3)
  ]}
 Set Stagger Progress State {nodes [
   n1 = Set Stagger Progress State!() -> (then, IsStagger?:bool=~, IsBroken?:bool=~) {is=title K2Node_FunctionEntry}
   > Delta Gate Wildcard!(execute:Exec, Variable:"Boolean (by ref)"=Get IsStaggerActive?() -> (IsStaggerActive?:bool), New Value:bool=@n1) -> (Success:Exec, Updated Variable:"Boolean (by ref)", Previous Variable:bool=~) {is=title K2Node_MacroInstance}
   > n2 = Branch(Condition=@n1.IsBroken?)
   > PlayAnimation(InAnimation=Get Anim_BrokenShield() -> (Anim_BrokenShield:Widget Animation Object Reference))
   > Set IsStaggerShowing?(IsStaggerShowing?:bool=false) -> (Output_Get:bool)
   > n3 = Set Timer by Event(
    Delegate=Create Event()
    Time=GetValueAtLevel(Input=Get StaggerDuration() -> (StaggerDuration:Scalable Float Structure), Level=0)
    bLooping=false
   )
   > Set Timer_HideStagger(Timer_HideStagger:Timer Handle Structure=@n3) -> (Output_Get:Timer Handle Structure)
   Clear and Invalidate Timer by Handle(execute=@n2.else, Handle=Get Timer_HideStagger() -> (Timer_HideStagger:Timer Handle Structure))
   > UpdateStaggerMeterVisibility(IsInstantHide=false)
  ]}
 SetShieldBlockingState {nodes [
   n1 = SetShieldBlockingState!() -> (then, IsBlocking?:bool=~) {is=title K2Node_FunctionEntry}
   > n2 = Delta Gate Wildcard(Variable=Get IsBlocking?() -> (IsBlocking?:bool), New Value=@n1) -> (Previous Variable=~)
   > UpdateStaggerMeterVisibility(IsInstantHide=true)
   > n3 = Branch(Condition=@n2)
   n4 = Get Anim_ShieldActivated() -> (Anim_ShieldActivated:Widget Animation Object Reference)
   > PlayAnimationForward(InAnimation=@n4)
   PlayAnimationReverse(execute=@n3.else, InAnimation=@n4)
  ]}
 UnbindBlockingBGA {nodes [
   UnbindBlockingBGA() {is=title K2Node_FunctionEntry}
   > n1 = Get!(execute) -> (then, else, Blocking BGA:BGA Ashton Shield Block Volume Object Reference)
   > Unbind Event from On Stagger Updated(self:BGA Ashton Shield Block Volume Object Reference=@n1, Delegate=Create Event()) {is=title K2Node_RemoveDelegate}
   > Set Blocking BGA()
  ]}
 UpdateForStagger {nodes [
   UpdateForStagger() {is=title K2Node_FunctionEntry}
   > Set NewProgress(NewProgress:double=GetProgress!() -> (Progress:double) {is=title K2Node_CallFunction}) -> (Output_Get:double)
   n1 = Get NewProgress() -> (NewProgress:double)
   n2 = "Nearly Equal (Float)"(A=@n1, B=0, ErrorTolerance=0)
   > Set Stagger Progress State!(execute, self, IsStagger?:bool=AND Boolean(A="float < float"(A=@n1, B=0.999), B=NOT Boolean(A=@n2)), IsBroken?:bool=@n2) -> (then) {is=title K2Node_CallFunction}
   > n3 = GetDynamicMaterial(self=Get M_UI_Reticle() -> (M_UI_Reticle:Image Object Reference))
   > SetScalarParameterValue(self=@n3, ParameterName=ProgressBar, Value=Get NewProgress())
  ]}
 UpdateStaggerMeterVisibility {nodes [
   UpdateStaggerMeterVisibility!() -> (then, IsInstantHide:bool=~) {is=title K2Node_FunctionEntry}
   n1 = Get IsBlockingConfirmed?() -> (IsBlockingConfirmed?:bool)
   n2 = Get IsBlocking?()
   > n3 = Delta Gate Wildcard(Variable=Get IsStaggerShowing?() -> (IsStaggerShowing?:bool), New Value=AND Boolean(A=AND Boolean(A=@n1, B=@n2), B=Get IsStaggerActive?())) -> (Previous Variable=~)
   > n4 = Branch(Condition=@n3)
   n5 = Get Anim_StaggerMeter() -> (Anim_StaggerMeter:Widget Animation Object Reference)
   > PlayAnimationForward(InAnimation=@n5)
   PlayAnimation(
    execute=@n4.else
    InAnimation=@n5
    StartAtTime=Select!(Option 0:float=0, Option 1:float=GetEndTime(self=@n5), Index:bool=Get IsInstantHide() -> (IsInstantHide:bool)) -> (ReturnValue:float)
    PlayMode=Reverse
   )
  ]}
]
variables [
 As B Ashton Shield Weap {type=B_Ashton_Shield_Weap_C, default=null}
 As Player Pawn Athena {type=PlayerPawn_Athena_C}
 AsyncActions {type="array<AbilityAsync>", default []}
 Blocking BGA {type=BGA_Ashton_Shield_BlockVolume_C, default=null}
 IsBlocking? {type=boolean}
 IsBlockingConfirmed? {type=boolean}
 IsStaggerActive? {type=boolean}
 IsStaggerShowing? {type=boolean}
 NewProgress {type=number}
 StaggerDuration {
  type=ScalableFloat
  default {registryType=None, value=1, curve {curveTable=/AshtonGameplay/DataTables/AshtonGameplayGameData, rowName=Default.Shield.StaggerDuration}}
 }
 TargetStateComponent {type=FortWeaponComponent_TargetState, default=null}
 Timer_HideStagger {type=TimerHandle}
]
widgets {
 info {
  inheritedWidgetCount=1
  parentClass=/Script/FortniteUI.FortWeaponReticleExtensionWidgetBase
  rootWidgetClass=&Overlay
  widgetCount=3
 }
 widgets [
  Overlay_0 {widgetClassPath=&Overlay, widget=Overlay_0}
  M_UI_Reticle {
   widgetClassPath=&Image
   bIsVariable=true
   parent=Overlay_0
   widget=M_UI_Reticle
   properties {brush {
     bIsDynamicallyLoaded=false
     drawAs=Image
     imageType=FullColor
     mirroring=NoMirror
     resourceName=None
     resourceObject=&/M_UI_Ashton_Reticle
     tiling=NoTile
     imageSize=<90 90>
     margin=<ltrb 0 0 0 0>
     outlineSettings {
      bUseBrushTransparency=false
      roundingType=HalfHeightRadius
      width=0
      color {colorUseRule=UseColor_Specified, specifiedColor=<rgba 0 0 0 0>}
      cornerRadii=<0 0 0 0>
     }
     tintColor {colorUseRule=UseColor_Specified, specifiedColor=<rgba 1 1 1 1>}
     uVRegion {bIsValid=false, max=<0 0>, min=<0 0>}
    }}
   slot {class=OverlaySlot, properties {horizontalAlignment=HAlign_Center, verticalAlignment=VAlign_Center}}
  }
  TutorialHighlight {widgetClassPath=&TutorialHighlight}
 ]
}
